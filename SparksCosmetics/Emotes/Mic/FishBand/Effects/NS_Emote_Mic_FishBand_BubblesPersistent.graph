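# Niagara system for persistent bubble sprites in the Fish Band mic emote, built from two probability-gated spawn-rate emitters.
# BodyBubbles spawns in a small sphere with SubUV animation, with velocity driven by a randomly scaled curve that rises faster over life.
# MouthBubbles launches from a local offset, slows with drag, floats upward and swells just before expiring.
# A SkeletalMesh user parameter is exposed, but no listed module reads it.
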NiagaraSystem /SparksCosmetics/Emotes/Mic/FishBand/Effects/NS_Emote_Mic_FishBand_BubblesPersistent
dependencies {
 usedDataInterfaces [&NiagaraDataInterfaceVectorCurve, &NiagaraDataInterfaceCurve]
 usedDynamicInputs []
 usedModules [&SystemState]
}
emitters {
 BodyBubbles {
  emitterUpdateScript [
   &EmitterState()
   &SpawnRate(Spawn Probability=0.1, SpawnRate=30, Use Spawn Probability=true)
  ]
  particleSpawnScript [
   &InitializeParticle(
    Color Maximum=<rgba 1 1 1 2>
    Color Minimum=<rgba 0.4887 0.544 1 0.5>
    Color Mode=::NewEnumerator2
    Sprite Size Mode=::NewEnumerator1
    Uniform Sprite Size Max=3
    Uniform Sprite Size Min=1
    UsePositionOffset=false
    Write Material Random=true
   )
   /Niagara/Modules/Update/SubUV/SubUVAnimation(SubUV Animation Mode=::NewEnumerator1)
   &ShapeLocation(MaskSpawnGroup=false, Override Seed=false, Sphere Radius=5, Sphere Surface Distribution=0.6381, UseOwnerScale=true)
   &DynamicMaterialParameters(
    Index 0 Param 1=1
    Index 0 Param 2=0.1
    Module.Param0WriteEnabled=true
    Module.Param1WriteEnabled=true
    Module.Param2WriteEnabled=true
    Module.Param3WriteEnabled=true
   )
  ]
  particleUpdateScript [
   &ParticleState()
   &SolveForcesAndVelocity(Clamp Velocity=false, Limit Acceleration=false)
   /SetVariables(Particles.Velocity=&VectorFromCurve(
     VectorCurve=DataInterface() {propertyValues {
       XCurve {keys [<at 0 50 RCIM_Cubic>, {arriveTangent=-150.7546, interpMode=RCIM_Cubic, leaveTangent=-150.7546, tangentMode=RCTM_User, time=0.5, value=20}]}
       YCurve {keys [<at 0 50 RCIM_Cubic>, {arriveTangent=-155.4714, interpMode=RCIM_Cubic, leaveTangent=-155.4714, tangentMode=RCTM_User, time=0.5, value=20}]}
       ZCurve {keys [<at 0 50>,<at 1 150>]}
      }}
     Scale Curve=&RandomRangeVector(Minimum=<-1 -1 0.5>, Maximum=<1 1 1>, Evaluation Type=::NewEnumerator1)
    )) {isSetParameters=true}
  ]
  rendererClasses [&NiagaraSpriteRendererProperties]
  renderers [{rendererClass=&NiagaraSpriteRendererProperties}]
 }
 MouthBubbles {
  emitterUpdateScript [
   &EmitterState()
   &SpawnRate(Spawn Probability=0.3, SpawnRate=10, Use Spawn Probability=true)
  ]
  particleSpawnScript [
   &InitializeParticle(
    Color Maximum=<rgba 1 1 1 2>
    Color Minimum=<rgba 0.4887 0.544 1 0.5>
    Color Mode=::NewEnumerator2
    Lifetime Min=0.5
    Lifetime Mode=::NewEnumerator1
    Position Offset=<5 0 -8>
    Position Offset Coordinate Space=::Local
    Sprite Size Mode=::NewEnumerator1
    Uniform Sprite Size Max=4
    Uniform Sprite Size Min=2
    UsePositionOffset=true
    Write Material Random=true
   )
   &DynamicMaterialParameters(
    Index 0 Param 1=1
    Index 0 Param 2=0.1
    Module.Param0WriteEnabled=true
    Module.Param1WriteEnabled=true
    Module.Param2WriteEnabled=true
    Module.Param3WriteEnabled=true
   )
   &AddVelocity(
    Module.UseConeFalloff=false
    Module.UseVelDistribution=false
    Override Seed=false
    Velocity Mode=::NewEnumerator2
    Velocity Speed=&RandomRangeFloat(Minimum=50, Maximum=75)
   )
  ]
  particleUpdateScript [
   &ParticleState()
   &Drag(Drag=1.5, Use Linear Drag=true, Use Rotational Drag=true)
   &AccelerationForce(Acceleration=<0 0 100>)
   &SolveForcesAndVelocity(Clamp Velocity=false, Limit Acceleration=false)
   &ScaleSpriteSize(Scale Factor=&Vector2DFromFloat(Value=&FloatFromCurve(FloatCurve=DataInterface() {propertyValues {Curve {keys [
          <at 0 0 RCIM_Cubic>
          {arriveTangent=0.002, interpMode=RCIM_Cubic, leaveTangent=-0.0489, tangentMode=RCTM_Break, time=0.25, value=0.9069}
          {arriveTangent=-0.077, interpMode=RCIM_Cubic, leaveTangent=-0.077, tangentMode=RCTM_User, time=0.9114, value=0.8888}
          {arriveTangent=4.9648, interpMode=RCIM_Cubic, leaveTangent=4.9648, tangentMode=RCTM_User, time=1, value=1.2}
         ]}}})))
  ]
  rendererClasses [&NiagaraSpriteRendererProperties]
  renderers [{rendererClass=&NiagaraSpriteRendererProperties}]
 }
}
summary {
 systemName=NS_Emote_Mic_FishBand_BubblesPersistent
 emitters [
  {emitterName=BodyBubbles, rendererClasses [&NiagaraSpriteRendererProperties]}
  {emitterName=MouthBubbles, rendererClasses [&NiagaraSpriteRendererProperties]}
 ]
 userVariables [User.SkeletalMesh {
   type=NiagaraDataInterfaceSkeletalMesh
   defaultValue {
    struct=NiagaraExt_VariableValue_DataInterface
    value {dataInterface=&self:NiagaraDataInterfaceSkeletalMesh_0}
   }
  }]
}
